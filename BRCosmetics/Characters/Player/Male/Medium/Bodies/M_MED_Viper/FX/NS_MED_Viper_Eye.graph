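# Niagara cosmetic effect for the Viper player character's head: two glowing cyan eye meshes plus blue and dark purple mouth smoke.
# A one-shot burst places the persistent eyes through a partition-indexed offset array.
# Continuous smoke sprites, with extra movement-based spawning for purple, use curve-driven flipbook frames, size and color, and downward acceleration.
# All three emitters write VisibilityTag from User.bisHeadHIdden to track the hidden-head state.
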NiagaraSystem /BRCosmetics/Characters/Player/Male/Medium/Bodies/M_MED_Viper/FX/NS_MED_Viper_Eye
dependencies {
 usedDataInterfaces [
  &NiagaraDataInterfaceArrayFloat3
  &NiagaraDataInterfaceCurve
  &NiagaraDataInterfaceSpriteRendererInfo
  &NiagaraDataInterfaceColorCurve
  &NiagaraDataInterfaceVector2DCurve
 ]
 usedDynamicInputs []
 usedModules [&SystemState]
}
emitters {
 Eye_Meshes {
  emitterUpdateScript [
   &EmitterState(
    Inactive Response=::NewEnumerator1
    Life Cycle Mode=::NewEnumerator1
    Loop Behavior=::NewEnumerator1
    Loop Duration Mode=::NewEnumerator1
    UseLoopDelay=false
   )
   &SpawnBurst_Instantaneous(Spawn Count=2, Use Spawn Probability=false, UseLoopCountLimit=true)
  ]
  particleSpawnScript [
   &PartitionParticles()
   &InitializeParticle(
    Color=<rgba 0.9279 6.7405 10 1>
    Lifetime=5
    Mesh Scale Mode=::NewEnumerator3
    Mesh Uniform Scale=0.05
    Position Offset=&SelectVectorFromArray(
     Direct Array Index=$Particles.PartitionParticles.Partition
     Vector Selection Array=DataInterface() {propertyValues {FloatData [<4 8 -3>, <4 8 3>]}}
    )
    UsePositionOffset=true
    Write Material Random=true
   )
  ]
  particleUpdateScript [
   &ParticleState(Kill Particles When Lifetime Has Elapsed=false, Let Infinitely Lived Particles Die When Emitter Deactivates=true)
   /SetVariables(Particles.VisibilityTag=&MakeCustomIntFromBool(Bool=$User.bisHeadHIdden, True INT=1, False INT=0)) {isSetParameters=true, module=SetVariables_#1}
  ]
  rendererClasses [&NiagaraMeshRendererProperties]
  renderers [{rendererClass=&NiagaraMeshRendererProperties, rendererIndex=0}]
 }
 Smoke_Mouth_Blue {
  emitterUpdateScript [
   &EmitterState()
   &SpawnRate(SpawnRate=&RandomRangeFloat(Minimum=5, Maximum=6), Use Spawn Probability=false)
  ]
  particleSpawnScript [
   &InitializeParticle(
    Color=&MakeLinearColorFromVectorAndFloat(
     "Vector (RGB)"=&MakeVectorFromLinearColorRGB(Color=<rgba 0.55 0.65 0.8 1>)
     "Float (Alpha)"=&RandomRangeFloat(Minimum=0.15, Maximum=0.3)
    )
    Lifetime Max=2.2
    Lifetime Min=1.7
    Lifetime Mode=::NewEnumerator1
    Position Offset=<0 2 0>
    Sprite Rotation Mode=::NewEnumerator2
    Sprite Size Mode=::NewEnumerator1
    Uniform Sprite Size Max=25
    Uniform Sprite Size Min=20
    UsePositionOffset=true
    Write Material Random=true
   )
   &CameraOffset(Camera Offset Amount=&RandomRangeFloat(Minimum=5, Maximum=10))
  ]
  particleUpdateScript [
   &ParticleState()
   &ScaleSpriteSize(Uniform Curve Sprite Scale=DataInterface() {propertyValues {Curve {keys [
        {arriveTangent=3.098, interpMode=RCIM_Cubic, leaveTangent=3.098, tangentMode=RCTM_User, time=0, value=0.2}
        {arriveTangentWeight=0.174, interpMode=RCIM_Cubic, leaveTangentWeight=0.174, tangentMode=RCTM_User, time=0.5, value=1}
        <at 1 0.5 RCIM_Cubic>
       ]}}})
   &SubUVAnimation(
    Direct Frame Index:NiagaraFloat=&FloatFromCurve(FloatCurve=DataInterface() {propertyValues {Curve {keys [{arriveTangent=20.9632, interpMode=RCIM_Cubic, leaveTangent=20.9632, tangentMode=RCTM_User, time=0, value=22}, <at 1 37.892 RCIM_Cubic>]}}})
    "Sprite Renderer "=DataInterface() {propertyValues {SpriteRenderer=&self:Smoke_Mouth_Blue_0.NiagaraSpriteRendererProperties_0}}
    SubUV Animation Mode:ENiagara_SubUVLookupModeV2=::NewEnumerator5
   )
   &ScaleColor(
    Scale Alpha=&FloatFromCurve(FloatCurve=DataInterface() {propertyValues {Curve {keys [
         {arriveTangent=2.866, interpMode=RCIM_Cubic, leaveTangent=2.866, tangentMode=RCTM_User, time=0, value=0}
         {arriveTangent=2.9072, interpMode=RCIM_Cubic, leaveTangent=2.9072, tangentMode=RCTM_User, time=0.25, value=1}
         {arriveTangent=-0.6667, interpMode=RCIM_Cubic, leaveTangent=-0.6667, time=0.9, value=1.8}
         <at 1 0.5 RCIM_Cubic>
        ]}}})
    Scale RGB=&MakeVectorFromLinearColorRGB(Color=&ColorFromCurve(ColorCurve=DataInterface() {propertyValues {
        AlphaCurve {keys []}
        BlueCurve {keys [<at 0.25 1 RCIM_Cubic>,<at 1 0.099 RCIM_Cubic>]}
        GreenCurve {keys [<at 0.25 0.1237 RCIM_Cubic>,<at 1 0.0788 RCIM_Cubic>]}
        RedCurve {keys [<at 0.25 0.0868 RCIM_Cubic>,<at 1 0.0696 RCIM_Cubic>]}
       }}))
    ScaleA=true
    ScaleRGB=true
   )
   &DynamicMaterialParameters(
    Index 0 Param 1=&RandomRangeFloat(Minimum=0, Maximum=1)
    Index 0 Param 2=&RandomRangeFloat(Minimum=1, Maximum=2)
    Module.Index0Float0Write=true
    Module.Index0Float1Write=true
    Module.Index0Float2Write=true
    Module.Index0Float3Write=false
   )
   &AccelerationForce(Acceleration=<0 0 -10>)
   &SolveForcesAndVelocity(Clamp Velocity=false, Limit Acceleration=false)
   /SetVariables(Particles.VisibilityTag=&MakeCustomIntFromBool(Bool=$User.bisHeadHIdden, True INT=1, False INT=0)) {isSetParameters=true, module=SetVariables_#1}
  ]
  rendererClasses [&NiagaraSpriteRendererProperties]
  renderers [{rendererClass=&NiagaraSpriteRendererProperties}]
 }
 Smoke_Mouth_Purple {
  emitterUpdateScript [
   &EmitterState()
   &SpawnRate(SpawnRate=8, Use Spawn Probability=false)
   &SpawnPerUnit(Module.Use Max Movement Threshold=true, Module.Use Movement Tolerance=false, Use Spawn Probability=true)
  ]
  particleSpawnScript [
   &InitializeParticle(
    Color=&MakeLinearColorFromVectorAndFloat(
     "Vector (RGB)"=&MakeVectorFromLinearColorRGB(Color=<rgba 0.0464 0 0.1406 1>)
     "Float (Alpha)"=&RandomRangeFloat(Minimum=0.2, Maximum=0.3)
    )
    Lifetime Min=1.5
    Lifetime Mode=::NewEnumerator1
    Position Offset=&RandomRangeVector(Minimum=<-2 5 -2>, Maximum=<2 10 2>)
    Sprite Rotation Angle=&RandomRangeFloat(Minimum=0.45, Maximum=0.55)
    Sprite Rotation Mode=::NewEnumerator1
    Sprite Size Mode=::NewEnumerator1
    Uniform Sprite Size Max=22
    Uniform Sprite Size Min=15
    UsePositionOffset=true
    Write Material Random=true
   )
   &CameraOffset(Camera Offset Amount=&RandomRangeFloat(Minimum=5, Maximum=7))
  ]
  particleUpdateScript [
   &ParticleState()
   &ScaleSpriteSize(
    Non-Uniform Curve Sprite Scale=DataInterface() {propertyValues {
      XCurve {keys [
        <at 0 0.1 RCIM_Cubic>
        {arriveTangent=1.3167, interpMode=RCIM_Cubic, leaveTangent=1.3167, time=0.2817, value=0.3202}
        {arriveTangent=0.9463, interpMode=RCIM_Cubic, leaveTangent=0.9463, time=0.6, value=0.89}
        <at 1 1 RCIM_Cubic>
       ]}
      YCurve {keys [
        {arriveTangent=4.8707, interpMode=RCIM_Cubic, leaveTangent=4.8707, tangentMode=RCTM_User, time=0, value=0.1}
        {arriveTangent=0.9858, interpMode=RCIM_Cubic, leaveTangent=0.9858, tangentMode=RCTM_User, time=0.2817, value=1.0116}
        {arriveTangent=-0.2626, interpMode=RCIM_Cubic, leaveTangent=-0.2626, tangentMode=RCTM_User, time=0.6, value=1.08}
        {interpMode=RCIM_Cubic, tangentMode=RCTM_User, time=1, value=1}
       ]}
     }}
    Scale Sprite Size Mode=::NewEnumerator3
   )
   &SubUVAnimation(
    Direct Frame Index:NiagaraFloat=&FloatFromCurve(FloatCurve=DataInterface() {propertyValues {Curve {keys [{arriveTangent=20.9632, interpMode=RCIM_Cubic, leaveTangent=20.9632, tangentMode=RCTM_User, time=0, value=4}, <at 1 19 RCIM_Cubic>]}}})
    "Sprite Renderer "=DataInterface() {propertyValues {SpriteRenderer=&self:Smoke_Mouth_Purple_0.NiagaraSpriteRendererProperties_0}}
    SubUV Animation Mode:ENiagara_SubUVLookupModeV2=::NewEnumerator5
   )
   &ScaleColor(
    Scale Alpha=&FloatFromCurve(FloatCurve=DataInterface() {propertyValues {Curve {keys [
         {arriveTangent=2.866, interpMode=RCIM_Cubic, leaveTangent=2.866, tangentMode=RCTM_User, time=0, value=0}
         {arriveTangent=2.9072, interpMode=RCIM_Cubic, leaveTangent=2.9072, tangentMode=RCTM_User, time=0.25, value=1}
         {arriveTangent=-0.6667, interpMode=RCIM_Cubic, leaveTangent=-0.6667, time=0.9, value=1.8}
         <at 1 0.5 RCIM_Cubic>
        ]}}})
    Scale RGB=&MakeVectorFromLinearColorRGB(Color=&ColorFromCurve(ColorCurve=DataInterface() {propertyValues {
        AlphaCurve {keys []}
        BlueCurve {keys [<at 0.25 5 RCIM_Cubic>,<at 1 0.099 RCIM_Cubic>]}
        GreenCurve {keys [<at 0.25 5 RCIM_Cubic>,<at 1 0.067 RCIM_Cubic>]}
        RedCurve {keys [<at 0.25 5 RCIM_Cubic>,<at 1 0.0499 RCIM_Cubic>]}
       }}))
    ScaleA=true
    ScaleRGB=true
   )
   &DynamicMaterialParameters(
    Index 0 Param 1=&RandomRangeFloat(Minimum=0, Maximum=1)
    Index 0 Param 2=&RandomRangeFloat(Minimum=1, Maximum=2)
    Module.Index0Float0Write=true
    Module.Index0Float1Write=true
    Module.Index0Float2Write=true
    Module.Index0Float3Write=false
   )
   &AccelerationForce(Acceleration=<0 0 -10>)
   &Drag(Drag=1.5, Use Linear Drag=true, Use Rotational Drag=true)
   &SolveForcesAndVelocity(Clamp Velocity=false, Limit Acceleration=false)
   /SetVariables(Particles.VisibilityTag=&MakeCustomIntFromBool(Bool=$User.bisHeadHIdden, True INT=1, False INT=0)) {isSetParameters=true, module=SetVariables_#1}
  ]
  rendererClasses [&NiagaraSpriteRendererProperties]
  renderers [{rendererClass=&NiagaraSpriteRendererProperties}]
 }
}
summary {
 systemName=NS_MED_Viper_Eye
 emitters [
  {emitterName=Eye_Meshes, rendererClasses [&NiagaraMeshRendererProperties]}
  {emitterName=Smoke_Mouth_Blue, rendererClasses [&NiagaraSpriteRendererProperties]}
  {emitterName=Smoke_Mouth_Purple, rendererClasses [&NiagaraSpriteRendererProperties]}
 ]
 userVariables [User.bisHeadHIdden {type=NiagaraBool}]
}
