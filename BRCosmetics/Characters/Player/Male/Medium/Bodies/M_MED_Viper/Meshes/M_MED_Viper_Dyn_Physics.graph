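# Physics asset for the dynamic accessories on the Viper medium male player body: two right-side grenades and three left-side pack pieces.
# Each grenade capsule and pack box hangs from its own small sphere base body through an angular-limited constraint, forming separate two-body pairs.
# Capsules cover only the right thigh, calf and foot, with a free knee and limited ankle; no constraint links this leg chain to the accessories.
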PhysicsAsset /BRCosmetics/Characters/Player/Male/Medium/Bodies/M_MED_Viper/Meshes/M_MED_Viper_Dyn_Physics
bodies {
 calf_r [{shapeType=Capsule, length=34.5, radius=6.25, center=<21.5271 1.6079 0.0001>, rotation=<rot 90 0 0>}]
 dyn_grenadeA_base_r [{shapeType=Sphere, radius=1}]
 dyn_grenadeA_r [{shapeType=Capsule, length=7.75, radius=2.75, center=<2.7435 0 0>, rotation=<rot 90 0 0>}]
 dyn_grenadeB_base_r [{shapeType=Sphere, radius=1}]
 dyn_grenadeB_r [{shapeType=Capsule, length=7.75, radius=2.75, center=<2.7435 0 0>, rotation=<rot 90 0 0>}]
 dyn_pack_A_base_l [{shapeType=Sphere, radius=1}]
 dyn_pack_A_l [{shapeType=Box, extentX=5.3226, extentY=6.5783, extentZ=18.367, center=<4.2468 1.2618 2.9903>, rotation=<rot 0 89.1094 -87.1106>}]
 dyn_pack_B_base_l [{shapeType=Sphere, radius=1}]
 dyn_pack_B_l [{shapeType=Box, extentX=5.4788, extentY=4.9964, extentZ=18.367, center=<3.2953 0.0943 2.6269>, rotation=<rot 0 88.2611 -90.1337>}]
 dyn_pack_C_base_l [{shapeType=Sphere, radius=1}]
 dyn_pack_C_l [{shapeType=Box, extentX=5.1275, extentY=5.2382, extentZ=18.367, center=<3.8644 -0.7993 2.5699>, rotation=<rot 0 86.6288 -86.1131>}]
 foot_r [{shapeType=Capsule, length=18, radius=6.25, center=<5.4559 -7.0476 0.0001>, rotation=<rot 90 26.565 113.1369>}]
 thigh_r [{shapeType=Capsule, length=25, radius=9.75, center=<17.2254 0.1288 -1.5877>, rotation=<rot 90 -0.0099 -0.0099>}]
}
constraints [
 {bone1Name=foot_r, bone2Name=calf_r, swing1LimitDegrees=45, swing2LimitDegrees=45, twistLimitDegrees=45}
 {bone1Name=dyn_grenadeA_r, bone2Name=dyn_grenadeA_base_r, swing1LimitDegrees=40, swing2LimitDegrees=30, twistLimitDegrees=20}
 {bone1Name=dyn_grenadeB_r, bone2Name=dyn_grenadeB_base_r, swing1LimitDegrees=40, swing2LimitDegrees=20, twistLimitDegrees=10}
 {bone1Name=dyn_pack_A_l, bone2Name=dyn_pack_A_base_l, swing1LimitDegrees=20, swing2LimitDegrees=30, twistLimitDegrees=10}
 {bone1Name=dyn_pack_B_l, bone2Name=dyn_pack_B_base_l, swing1LimitDegrees=20, swing2LimitDegrees=30, twistLimitDegrees=10}
 {bone1Name=dyn_pack_C_l, bone2Name=dyn_pack_C_base_l, swing1LimitDegrees=20, swing2LimitDegrees=30, twistLimitDegrees=10}
 {
  bone1Name=calf_r
  bone2Name=thigh_r
  swing1LimitDegrees=45
  swing1Motion=Free
  swing2LimitDegrees=45
  swing2Motion=Free
  twistLimitDegrees=45
  twistMotion=Free
 }
]
