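# Animation Blueprint for the medium male Viper body cosmetic that drives its secondary bones for grenades, packs, belt and leg straps.
# It copies the parent mesh's pose, blends those bones between pelvis, spine and thighs with weighted constraints, then applies additive corrective offsets scaled by leg angles.
# Two RigidBody nodes then simulate physics relative to root and right thigh.
# The EventGraph measures thigh angles against pelvis and sets an LOD-based rigid-body-disable flag, which the AnimGraph leaves unused.
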AnimBlueprint /BRCosmetics/Characters/Player/Male/Medium/Bodies/M_MED_Viper/Meshes/M_MED_Viper_AnimBP
target_skeleton=&/M_MED_Viper_Skeleton
graphs [
 AnimGraph {nodes [
   n1 = "Constraint - dyn_grenadeA_base_r (PPP)"!(
    ComponentPose=Local To Component(LocalPose=Copy Pose From Mesh() {bUseAttachedParent=true})
    Alpha=1
    ConstraintWeights_0=0.1924
    ConstraintWeights_1:float=0.0283
    ConstraintWeights_2:float=0.7793
   ) -> (Pose) {
    alphaBoolBlend {blendOption=Linear}
    boneToModify {boneName=dyn_grenadeA_base_r}
    constraintSetup [{targetBone {boneName=spine_01}}, {targetBone {boneName=thigh_twist_01_l}}, {targetBone {boneName=pelvis}}]
    constraintWeights [1,1,1]
   }
   n2 = "Constraint - dyn_grenadeB_base_r (PPP)"!(ComponentPose=@n1, Alpha=1, ConstraintWeights_0=0.1462, ConstraintWeights_1:float=0.0278, ConstraintWeights_2:float=0.826) -> (Pose) {
    alphaBoolBlend {blendOption=Linear}
    boneToModify {boneName=dyn_grenadeB_base_r}
    constraintSetup [{targetBone {boneName=spine_01}}, {targetBone {boneName=thigh_twist_01_l}}, {targetBone {boneName=pelvis}}]
    constraintWeights [1,1,1]
   }
   n3 = "Constraint - dyn_pack_A_base_l (PPP)"!(ComponentPose=@n2, Alpha=1, ConstraintWeights_0=0.8003, ConstraintWeights_1:float=0.0002, ConstraintWeights_2:float=0.1995) -> (Pose) {
    alphaBoolBlend {blendOption=Linear}
    boneToModify {boneName=dyn_pack_A_base_l}
    constraintSetup [{targetBone {boneName=pelvis}}, {targetBone {boneName=thigh_twist_01_l}}, {targetBone {boneName=thigh_twist_01_r}}]
    constraintWeights [1,1,1]
   }
   n4 = "Constraint - dyn_pack_B_base_l (PP)"!(ComponentPose=@n3, Alpha=1, ConstraintWeights_0=0.6029, ConstraintWeights_1:float=0.3971) -> (Pose) {
    alphaBoolBlend {blendOption=Linear}
    boneToModify {boneName=dyn_pack_B_base_l}
    constraintSetup [{targetBone {boneName=pelvis}}, {targetBone {boneName=thigh_twist_01_r}}]
    constraintWeights [1,1]
   }
   n5 = "Constraint - dyn_pack_C_base_l (PP)"!(ComponentPose=@n4, Alpha=1, ConstraintWeights_0=0.733, ConstraintWeights_1:float=0.267) -> (Pose) {
    alphaBoolBlend {blendOption=Linear}
    boneToModify {boneName=dyn_pack_C_base_l}
    constraintSetup [{targetBone {boneName=pelvis}}, {targetBone {boneName=thigh_twist_01_r}}]
    constraintWeights [1,1]
   }
   n6 = "Constraint - dyn_base_legstrapa_A_base_r (PP)"!(ComponentPose=@n5, Alpha=1, ConstraintWeights_0=0.7431, ConstraintWeights_1:float=0.2569) -> (Pose) {
    alphaBoolBlend {blendOption=Linear}
    boneToModify {boneName=dyn_base_legstrapa_A_base_r}
    constraintSetup [{targetBone {boneName=pelvis}}, {targetBone {boneName=thigh_twist_01_r}}]
    constraintWeights [1,1]
   }
   n7 = "Constraint - dyn_base_legstrapb_A_base_r (PPP)"!(ComponentPose=@n6, Alpha=1, ConstraintWeights_0=0.7398, ConstraintWeights_1:float=0, ConstraintWeights_2:float=0.2601) -> (Pose) {
    alphaBoolBlend {blendOption=Linear}
    boneToModify {boneName=dyn_base_legstrapb_A_base_r}
    constraintSetup [{targetBone {boneName=pelvis}}, {targetBone {boneName=thigh_twist_01_l}}, {targetBone {boneName=thigh_twist_01_r}}]
    constraintWeights [1,1,1]
   }
   n8 = "Transform (Modify) Bone - Bone: dyn_belt_adj"(ComponentPose=@n7, Alpha=Get legUp_L() -> (legUp_L:double), Translation="-6,0,1", Rotation="5,0,0", Scale="1,1,1") {
    rotationMode=BMM_Additive
    rotationSpace=BCS_BoneSpace
    translationMode=BMM_Additive
    translationSpace=BCS_BoneSpace
    alphaBoolBlend {blendOption=Linear}
    alphaScaleBiasClamp {bMapRange=true, inRange {max=-60, min=-100}, outRange {max=0, min=1}}
    boneToModify {boneName=dyn_belt_adj}
    rotation {pitch=5}
    translation {x=-6, z=1}
   }
   n9 = Get legUp_R() -> (legUp_R:double)
   n10 = "Transform (Modify) Bone - Bone: dyn_pack_B_base_l"(ComponentPose=@n8, Alpha=@n9, Translation="0,0,0", Rotation="0,20,10", Scale="1,1,1") {
    rotationMode=BMM_Additive
    rotationSpace=BCS_BoneSpace
    translationMode=BMM_Additive
    translationSpace=BCS_BoneSpace
    alphaBoolBlend {blendOption=Linear}
    alphaScaleBiasClamp {bMapRange=true, inRange {max=0, min=-100}, outRange {max=0, min=1}}
    boneToModify {boneName=dyn_pack_B_base_l}
    rotation {roll=10, yaw=20}
   }
   n11 = "Transform (Modify) Bone - Bone: dyn_pack_C_base_l"(ComponentPose=@n10, Alpha=@n9, Translation="0,0,2", Rotation="0,10,30", Scale="1,1,1") {
    rotationMode=BMM_Additive
    rotationSpace=BCS_BoneSpace
    translationMode=BMM_Additive
    translationSpace=BCS_BoneSpace
    alphaBoolBlend {blendOption=Linear}
    alphaScaleBiasClamp {bMapRange=true, inRange {max=0, min=-100}, outRange {max=0, min=1}}
    boneToModify {boneName=dyn_pack_C_base_l}
    rotation {roll=30, yaw=10}
    translation {z=2}
   }
   n12 = Get legSide_R() -> (legSide_R:double)
   n13 = "Transform (Modify) Bone - Bone: dyn_pack_B_base_l"(ComponentPose=@n11, Alpha=@n12, Translation="0,0,1.5", Rotation="10,0,0", Scale="1,1,1") {
    rotationMode=BMM_Additive
    rotationSpace=BCS_BoneSpace
    translationMode=BMM_Additive
    translationSpace=BCS_BoneSpace
    alphaBoolBlend {blendOption=Linear}
    alphaScaleBiasClamp {bMapRange=true, inRange {max=40}}
    boneToModify {boneName=dyn_pack_B_base_l}
    rotation {pitch=10}
    translation {z=1.5}
   }
   n14 = "Transform (Modify) Bone - Bone: dyn_pack_C_base_l"(ComponentPose=@n13, Alpha=@n12, Translation="0,0,4", Rotation="30,0,0", Scale="1,1,1") {
    rotationMode=BMM_Additive
    rotationSpace=BCS_BoneSpace
    translationMode=BMM_Additive
    translationSpace=BCS_BoneSpace
    alphaBoolBlend {blendOption=Linear}
    alphaScaleBiasClamp {bMapRange=true, inRange {max=40}}
    boneToModify {boneName=dyn_pack_C_base_l}
    rotation {pitch=30}
    translation {z=4}
   }
   n15 = "Constraint - dyn_base_legstrapa_A_base_r (R)"(ComponentPose=@n14, Alpha=1, ConstraintWeights_0=1) {
    alphaBoolBlend {blendOption=Linear}
    boneToModify {boneName=dyn_base_legstrapa_A_base_r}
    constraintSetup [{transformType=Rotation, targetBone {boneName=thigh_twist_01_r}}]
    constraintWeights [1]
   }
   n16 = "Transform (Modify) Bone - Bone: dyn_base_legstrapa_A_base_r"(ComponentPose=@n15, Alpha=@n9, Translation="0,0,4", Rotation="0,0,0", Scale="1,1,1") {
    rotationMode=BMM_Additive
    rotationSpace=BCS_BoneSpace
    translationMode=BMM_Additive
    translationSpace=BCS_BoneSpace
    alphaBoolBlend {blendOption=Linear}
    alphaScaleBiasClamp {bMapRange=true, inRange {max=0, min=-100}, outRange {max=0, min=1}}
    boneToModify {boneName=dyn_base_legstrapa_A_base_r}
    translation {z=4}
   }
   n17 = RigidBody(ComponentPose=@n16) {
    bClampLinearTranslationLimitToRefPose=true
    bOverrideWorldGravity=true
    overridePhysicsAsset=&/M_MED_Viper_Dyn_Physics
    simulationSpace=BaseBoneSpace
    alphaBoolBlend {blendOption=Linear}
    baseBoneRef {boneName=root}
    componentLinearAccScale=<0.1 0.1 0.1>
    componentLinearVelScale=<0.1 0.1 0.1>
    overrideWorldGravity {z=-980}
   }
   n18 = RigidBody(ComponentPose=@n17) {
    bClampLinearTranslationLimitToRefPose=true
    bOverrideWorldGravity=true
    overridePhysicsAsset=&/M_MED_Viper_Dyn_Belt_Physics1
    simulationSpace=BaseBoneSpace
    alphaBoolBlend {blendOption=Linear}
    baseBoneRef {boneName=thigh_r}
    componentLinearAccScale=<0.3 0.3 0.3>
    componentLinearVelScale=<0.3 0.3 0.3>
    overrideWorldGravity {z=-980}
   }
   Output Pose(Result=Component To Local(ComponentPose=@n18)) {layerGroup=None}
   Trail controller - Bone: dyn_legstrapb_C_r(Alpha=0) {
    relaxationSpeedScale=5
    rotationLimits [{limitMax=<180 180 180>}, {limitMax=<180 180 180>}]
    rotationOffsets [<0 0 0>, <0 0 0>]
    trailBone {boneName=dyn_legstrapb_C_r}
   }
  ]}
 EventGraph {nodes [
   Event BlueprintUpdateAnimation()
   > n1 = Sequence!(execute) -> (then_0, then_1, then_2:Exec)
   n2 = Get CurrentLOD()
   n3 = Get bRigidBodyEnabled()
   > Set isRigidBodyDisabled(isRigidBodyDisabled:bool=OR Boolean(A=NOT Boolean(A=@n3), B="integer > integer"(A=@n2, B=1) {is=title K2Node_CallFunction})) -> (Output_Get:bool)
   n4 = GetOwnerSkeletalMeshComponent()
   n5 = GetDeltaTransformFromRefPose(self=@n4, BoneName=thigh_l, BaseName=pelvis)
   Set legUp_L(execute=@n1.then_1, legUp_L:double=BreakRotator(InRot=BreakTransform(InTransform=@n5).Rotation).Yaw) -> (Output_Get:double)
   n6 = BreakRotator(InRot=BreakTransform(InTransform=GetDeltaTransformFromRefPose(self=@n4, BoneName=thigh_r, BaseName=pelvis)).Rotation)
   Set legUp_R(execute=@n1.then_2, legUp_R:double=@n6.Yaw) -> (Output_Get:double)
   > Set legSide_R(legSide_R:double=@n6.Pitch) -> (Output_Get:double)
   TryGetPawnOwner()
  ]}
]
variables [
 isRigidBodyDisabled {type=boolean}
 legSide_R {type=number}
 legUp_L {type=number}
 legUp_R {type=number}
]
