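# Physics asset for the Viper medium male body's face accessory mesh, defining bodies and joints for the motion of its hood.
# Front and back hood flaps are two-box chains with limited swing and twist, hung from small sphere base bodies linked to spine_05.
# A capsule hood peak has tight limits to the head.
# Spine capsules provide torso volume, while head, upper-arm, base and spine links leave angular motion free.
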PhysicsAsset /BRCosmetics/Characters/Player/Male/Medium/Bodies/M_MED_Viper/Meshes/Parts/M_MED_Viper_FaceAcc_Physics
bodies {
 dyn_hood_back_a [{shapeType=Box, extentX=11, extentY=22.5, extentZ=4.5, center=<6.2682 0 0>}]
 dyn_hood_back_a_base [{shapeType=Sphere, radius=1}]
 dyn_hood_back_b [{shapeType=Box, extentX=9.5, extentY=22.5, extentZ=4.5, center=<4.5946 0 0>}]
 dyn_hood_back_peak [{shapeType=Capsule, length=5, radius=3.005, center=<2.7183 0 0.0001>, rotation=<rot 90 0 90>}]
 dyn_hood_front_a [{shapeType=Box, extentX=9.5, extentY=22.5, extentZ=3.5, center=<4.3459 0.0003 0>}]
 dyn_hood_front_a_base [{shapeType=Sphere, radius=1}]
 dyn_hood_front_b [{shapeType=Box, extentX=9.5, extentY=22.5, extentZ=2.5, center=<4.3459 0.0003 0>}]
 dyn_uparm_base_l [{shapeType=Sphere, radius=1}]
 dyn_uparm_base_r [{shapeType=Sphere, radius=1}]
 head [{shapeType=Sphere, radius=1}]
 spine_03 [{shapeType=Capsule, length=15.25, radius=9.5, center=<0 2.0889 0>}]
 spine_04 [{shapeType=Capsule, length=27.25, radius=12.25, center=<0 0.3765 0>}]
 spine_05 [{shapeType=Capsule, length=33.5, radius=12, center=<-1.7525 0.6527 0>}]
}
constraints [
 {
  bone1Name=spine_05
  bone2Name=spine_04
  swing1LimitDegrees=45
  swing1Motion=Free
  swing2LimitDegrees=45
  swing2Motion=Free
  twistLimitDegrees=45
  twistMotion=Free
 }
 {
  bone1Name=spine_04
  bone2Name=spine_03
  swing1LimitDegrees=45
  swing1Motion=Free
  swing2LimitDegrees=45
  swing2Motion=Free
  twistLimitDegrees=45
  twistMotion=Free
 }
 {bone1Name=dyn_hood_back_a, bone2Name=dyn_hood_back_a_base, swing1LimitDegrees=5, swing2LimitDegrees=40, twistLimitDegrees=5}
 {bone1Name=dyn_hood_back_b, bone2Name=dyn_hood_back_a, swing1LimitDegrees=10, swing2LimitDegrees=40, twistLimitDegrees=5}
 {bone1Name=dyn_hood_front_a, bone2Name=dyn_hood_front_a_base, swing1LimitDegrees=5, swing2LimitDegrees=45, twistLimitDegrees=5}
 {bone1Name=dyn_hood_front_b, bone2Name=dyn_hood_front_a, swing1LimitDegrees=10, swing2LimitDegrees=45, twistLimitDegrees=5}
 {
  bone1Name=dyn_hood_back_a_base
  bone2Name=spine_05
  swing1LimitDegrees=45
  swing1Motion=Free
  swing2LimitDegrees=45
  swing2Motion=Free
  twistLimitDegrees=45
  twistMotion=Free
 }
 {
  bone1Name=dyn_hood_front_a_base
  bone2Name=spine_05
  swing1LimitDegrees=45
  swing1Motion=Free
  swing2LimitDegrees=45
  swing2Motion=Free
  twistLimitDegrees=45
  twistMotion=Free
 }
 {
  bone1Name=dyn_uparm_base_l
  bone2Name=spine_05
  swing1LimitDegrees=45
  swing1Motion=Free
  swing2LimitDegrees=45
  swing2Motion=Free
  twistLimitDegrees=45
  twistMotion=Free
 }
 {
  bone1Name=dyn_uparm_base_r
  bone2Name=spine_05
  swing1LimitDegrees=45
  swing1Motion=Free
  swing2LimitDegrees=45
  swing2Motion=Free
  twistLimitDegrees=45
  twistMotion=Free
 }
 {bone1Name=dyn_hood_back_peak, bone2Name=head, swing1LimitDegrees=15, swing2LimitDegrees=5, twistLimitDegrees=10}
 {
  bone1Name=head
  bone2Name=spine_05
  swing1LimitDegrees=45
  swing1Motion=Free
  swing2LimitDegrees=45
  swing2Motion=Free
  twistLimitDegrees=45
  twistMotion=Free
 }
]
