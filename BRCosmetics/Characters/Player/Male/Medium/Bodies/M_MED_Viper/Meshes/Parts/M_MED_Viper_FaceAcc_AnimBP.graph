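# Anim Blueprint for the Viper male face-accessory part, adding secondary motion to hood and upper-arm bones.
# It copies the owner mesh's pose, offsets bones, then runs AnimDynamics, faded by arm pitch for arm bones, and RigidBody physics.
# Constraints pin bones to body bones when rigid body is off or LOD exceeds 1, and lock back-hood rotation when a backpack is detected.
# Forward/backward movement values and a hood-peak dynamics node are unused.
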AnimBlueprint /BRCosmetics/Characters/Player/Male/Medium/Bodies/M_MED_Viper/Meshes/Parts/M_MED_Viper_FaceAcc_AnimBP
target_skeleton=&/M_MED_Viper_FaceAcc_Skeleton
graphs [
 AnimGraph {nodes [
   Anim Dynamics - Bone: dyn_hood_back_peak(Alpha=1, ExternalForce="0,0,0") {
    angularDampingOverride=1
    angularSpringConstant=1500
    bAngularSpring=true
    bGravityOverrideInSimSpace=true
    bOverrideAngularDamping=true
    gravityScale=0.35
    simulationSpace=RootRelative
    alphaBoolBlend {blendOption=Linear}
    boundBone {boneName=dyn_hood_back_peak}
    physicsBodyDefinitions [{
      sphereCollisionRadius=0
      boundBone {boneName=dyn_hood_back_peak}
      boxExtents=<10 20 20>
      constraintSetup {angularLimitsMax=<0 5 20>, angularLimitsMin=<0 -5 -20>, angularTarget=<0 0 0>}
      localJointOffset=<10 0 0>
     }]
   }
   n1 = Get MeshToCopy() -> (MeshToCopy:Skeletal Mesh Component Object Reference)
   n2 = "Transform (Modify) Bone - Bone: dyn_hood_top"(
    ComponentPose=Local To Component(LocalPose=Copy Pose From Mesh(SourceMeshComponent=@n1))
    Alpha=1
    Translation="2,0,0"
    Rotation="0,0,0"
    Scale="1,1,1"
   ) {
    rotationMode=BMM_Additive
    rotationSpace=BCS_BoneSpace
    translationMode=BMM_Additive
    translationSpace=BCS_BoneSpace
    alphaBoolBlend {blendOption=Linear}
    boneToModify {boneName=dyn_hood_top}
    translation {x=2}
   }
   n3 = "Transform (Modify) Bone - Bone: dyn_hood_front_a_base"(ComponentPose=@n2, Alpha=1, Translation="0,0,1.5", Rotation="0,0,0", Scale="1,1,1") {
    rotationMode=BMM_Additive
    rotationSpace=BCS_BoneSpace
    translationMode=BMM_Additive
    translationSpace=BCS_BoneSpace
    alphaBoolBlend {blendOption=Linear}
    boneToModify {boneName=dyn_hood_front_a_base}
    translation {z=1.5}
   }
   n4 = "Transform (Modify) Bone - Bone: dyn_hood_back_a_base"(ComponentPose=@n3, Alpha=1, Translation="0,0,2", Rotation="0,0,0", Scale="1,1,1") {
    rotationMode=BMM_Additive
    rotationSpace=BCS_BoneSpace
    translationMode=BMM_Additive
    translationSpace=BCS_BoneSpace
    alphaBoolBlend {blendOption=Linear}
    boneToModify {boneName=dyn_hood_back_a_base}
    translation {z=2}
   }
   n5 = "Transform (Modify) Bone - Bone: dyn_uparm_base_r"(ComponentPose=@n4, Alpha=1, Translation="0,0,-1", Rotation="0,0,0", Scale="1,1,1") {
    rotationMode=BMM_Additive
    rotationSpace=BCS_BoneSpace
    translationMode=BMM_Additive
    translationSpace=BCS_BoneSpace
    alphaBoolBlend {blendOption=Linear}
    boneToModify {boneName=dyn_uparm_base_r}
    translation {z=-1}
   }
   n6 = "Transform (Modify) Bone - Bone: dyn_uparm_base_l"(ComponentPose=@n5, Alpha=1, Translation="0,0,1", Rotation="0,0,0", Scale="1,1,1") {
    rotationMode=BMM_Additive
    rotationSpace=BCS_BoneSpace
    translationMode=BMM_Additive
    translationSpace=BCS_BoneSpace
    alphaBoolBlend {blendOption=Linear}
    boneToModify {boneName=dyn_uparm_base_l}
    translation {z=1}
   }
   n7 = Anim Dynamics - Bone: dyn_hood_top(ComponentPose=@n6, Alpha=1, ExternalForce="0,0,0") {
    angularDampingOverride=1
    angularSpringConstant=20000
    bAngularSpring=true
    bGravityOverrideInSimSpace=true
    bOverrideAngularDamping=true
    gravityScale=0.35
    simulationSpace=RootRelative
    alphaBoolBlend {blendOption=Linear}
    boundBone {boneName=dyn_hood_top}
    componentLinearAccScale=<0.5 0.5 0.5>
    componentLinearVelScale=<0.5 0.5 0.5>
    physicsBodyDefinitions [{
      sphereCollisionRadius=0
      boundBone {boneName=dyn_hood_top}
      boxExtents=<10 20 20>
      constraintSetup {angularLimitsMax=<0 5 5>, angularLimitsMin=<0 -5 -5>, angularTarget=<1 0 0>}
      localJointOffset=<5 0 0>
     }]
   }
   n8 = Anim Dynamics - Bone: dyn_hood_l(ComponentPose=@n7, Alpha=1, ExternalForce="0,0,0") {
    angularDampingOverride=1
    angularSpringConstant=50000
    bAngularSpring=true
    bGravityOverrideInSimSpace=true
    bOverrideAngularDamping=true
    gravityScale=0.35
    simulationSpace=RootRelative
    alphaBoolBlend {blendOption=Linear}
    boundBone {boneName=dyn_hood_l}
    componentLinearAccScale=<0.5 0.5 0.5>
    componentLinearVelScale=<0.5 0.5 0.5>
    physicsBodyDefinitions [{
      sphereCollisionRadius=0
      boundBone {boneName=dyn_hood_l}
      boxExtents=<10 20 20>
      constraintSetup {angularLimitsMax=<0 0 0>, angularLimitsMin=<0 0 -25>, angularTarget=<0 0 0>}
      localJointOffset=<5 0 0>
     }]
   }
   n9 = Anim Dynamics - Bone: dyn_hood_r(ComponentPose=@n8, Alpha=1, ExternalForce="0,0,0") {
    angularDampingOverride=1
    angularSpringConstant=50000
    bAngularSpring=true
    bGravityOverrideInSimSpace=true
    bOverrideAngularDamping=true
    gravityScale=0.35
    simulationSpace=RootRelative
    alphaBoolBlend {blendOption=Linear}
    boundBone {boneName=dyn_hood_r}
    componentLinearAccScale=<0.5 0.5 0.5>
    componentLinearVelScale=<0.5 0.5 0.5>
    physicsBodyDefinitions [{
      sphereCollisionRadius=0
      boundBone {boneName=dyn_hood_r}
      boxExtents=<10 20 20>
      constraintSetup {angularLimitsMax=<0 0 0>, angularLimitsMin=<0 0 -25>, angularTarget=<0 0 0>}
      localJointOffset=<-5 0 0>
     }]
   }
   n10 = Anim Dynamics - Bone: dyn_hood_front_md(ComponentPose=@n9, Alpha=1, ExternalForce="0,0,0") {
    angularDampingOverride=1
    angularSpringConstant=10000
    bAngularSpring=true
    bGravityOverrideInSimSpace=true
    bOverrideAngularDamping=true
    gravityScale=0.35
    simulationSpace=RootRelative
    alphaBoolBlend {blendOption=Linear}
    boundBone {boneName=dyn_hood_front_md}
    physicsBodyDefinitions [{
      sphereCollisionRadius=0
      boundBone {boneName=dyn_hood_front_md}
      boxExtents=<10 10 10>
      constraintSetup {angularLimitsMax=<0 5 10>, angularLimitsMin=<0 -5 -10>, angularTarget=<1 0 0>}
      localJointOffset=<5 0 0>
     }]
   }
   n11 = Anim Dynamics - Bone: dyn_uparm_l(ComponentPose=@n10, Alpha=Get armUp_L() -> (armUp_L:double), ExternalForce="0,0,0") {
    angularDampingOverride=1
    angularSpringConstant=1500
    bAngularSpring=true
    bGravityOverrideInSimSpace=true
    bOverrideAngularDamping=true
    gravityScale=0.35
    numSolverIterationsPreUpdate=8
    simulationSpace=RootRelative
    alphaBoolBlend {blendOption=Linear}
    alphaScaleBiasClamp {bMapRange=true, inRange {max=20}, outRange {max=0, min=1}}
    boundBone {boneName=dyn_uparm_l}
    componentLinearAccScale {z=0.4}
    componentLinearVelScale {z=0.4}
    physicsBodyDefinitions [{
      sphereCollisionRadius=0
      boundBone {boneName=dyn_uparm_l}
      boxExtents=<10 20 20>
      constraintSetup {angularLimitsMax=<0 0 0>, angularLimitsMin=<0 -50 0>, angularTarget=<0 0 0>}
      localJointOffset=<10 0 0>
     }]
   }
   n12 = Apply Limits!(ComponentPose:Component Space Pose Link Structure=@n11, Alpha:float=1) -> (Pose:Component Space Pose Link Structure) {
    is=title AnimGraphNode_ApplyLimits
    alphaBoolBlend {blendOption=Linear}
    angularOffsets [<0 0 0>]
    angularRangeLimits [{bone {boneName=dyn_uparm_l}, limitMax=<0 180 0>, limitMin=<0 -180 0>}]
   }
   n13 = Anim Dynamics - Bone: dyn_uparm_r(ComponentPose=@n12, Alpha=Get armUp_R() -> (armUp_R:double), ExternalForce="0,0,0") {
    angularDampingOverride=1
    angularSpringConstant=1500
    bAngularSpring=true
    bGravityOverrideInSimSpace=true
    bOverrideAngularDamping=true
    gravityScale=0.35
    numSolverIterationsPreUpdate=8
    simulationSpace=RootRelative
    alphaBoolBlend {blendOption=Linear}
    alphaScaleBiasClamp {bMapRange=true, inRange {max=2}, outRange {max=0, min=1}}
    boundBone {boneName=dyn_uparm_r}
    componentLinearAccScale {z=0.4}
    componentLinearVelScale {z=0.4}
    physicsBodyDefinitions [{
      sphereCollisionRadius=0
      boundBone {boneName=dyn_uparm_r}
      boxExtents=<10 20 20>
      constraintSetup {angularLimitsMax=<0 0 0>, angularLimitsMin=<0 -50 0>, angularTarget=<0 0 0>}
      localJointOffset=<-10 0 0>
     }]
   }
   n14 = Apply Limits(ComponentPose=@n13, Alpha=1) {
    alphaBoolBlend {blendOption=Linear}
    angularOffsets [<0 0 0>]
    angularRangeLimits [{bone {boneName=dyn_uparm_r}, limitMax=<0 180 0>, limitMin=<0 -180 0>}]
   }
   n15 = Get noDynamic() -> (noDynamic:double)
   n16 = "Constraint - dyn_uparm_base_l (P)"(ComponentPose=@n14, Alpha=@n15, ConstraintWeights_0=1) {
    alphaBoolBlend {blendOption=Linear}
    boneToModify {boneName=dyn_uparm_base_l}
    constraintSetup [{targetBone {boneName=upperarm_l}}]
    constraintWeights [1]
   }
   n17 = "Constraint - dyn_uparm_base_r (P)"(ComponentPose=@n16, Alpha=@n15, ConstraintWeights_0=1) {
    alphaBoolBlend {blendOption=Linear}
    boneToModify {boneName=dyn_uparm_base_r}
    constraintSetup [{targetBone {boneName=upperarm_r}}]
    constraintWeights [1]
   }
   n18 = Get noDynamic()
   n19 = "Constraint - dyn_hood_front_a (PPP)"!(ComponentPose=@n17, Alpha=@n18, ConstraintWeights_0=0.1504, ConstraintWeights_1:float=0.523, ConstraintWeights_2:float=0.3265) -> (Pose) {
    alphaBoolBlend {blendOption=Linear}
    boneToModify {boneName=dyn_hood_front_a}
    constraintSetup [{targetBone {boneName=spine_03}}, {targetBone {boneName=spine_04}}, {targetBone {boneName=spine_05}}]
    constraintWeights [1,1,1]
   }
   n20 = "Constraint - dyn_hood_front_b (PPPP)"!(
    ComponentPose=@n19
    Alpha=@n18
    ConstraintWeights_0=0.0317
    ConstraintWeights_1:float=0.4279
    ConstraintWeights_2:float=0.4593
    ConstraintWeights_3:float=0.081
   ) -> (Pose) {
    alphaBoolBlend {blendOption=Linear}
    boneToModify {boneName=dyn_hood_front_b}
    constraintSetup [{targetBone {boneName=clavicle_r}}, {targetBone {boneName=spine_03}}, {targetBone {boneName=spine_04}}, {targetBone {boneName=spine_05}}]
    constraintWeights [1,1,1,1]
   }
   n21 = RigidBody(ComponentPose=@n20) {
    bClampLinearTranslationLimitToRefPose=true
    bOverrideWorldGravity=true
    overridePhysicsAsset=&/M_MED_Viper_FaceAcc_Physics
    simulationSpace=BaseBoneSpace
    alphaBoolBlend {blendOption=Linear}
    baseBoneRef {boneName=spine_02}
    componentLinearAccScale=<0.3 0.3 0.4>
    componentLinearVelScale=<0.3 0.3 0.4>
    overrideWorldGravity {z=-980}
   }
   n22 = Get IsBackpackOn() -> (IsBackpackOn:double)
   n23 = "Constraint - dyn_hood_back_a (R)"(ComponentPose=@n21, Alpha=@n22, ConstraintWeights_0=1) {
    alphaBoolBlend {blendOption=Linear}
    boneToModify {boneName=dyn_hood_back_a}
    constraintSetup [{transformType=Rotation, targetBone {boneName=spine_05}}]
    constraintWeights [1]
   }
   n24 = "Constraint - dyn_hood_back_b (R)"(ComponentPose=@n23, Alpha=@n22, ConstraintWeights_0=1) {
    alphaBoolBlend {blendOption=Linear}
    boneToModify {boneName=dyn_hood_back_b}
    constraintSetup [{transformType=Rotation, targetBone {boneName=spine_05}}]
    constraintWeights [1]
   }
   Output Pose(Result=Component To Local(ComponentPose=@n24)) {layerGroup=None}
  ]}
 EventGraph {nodes [
   Event BlueprintInitializeAnimation()
   > n1 = Sequence()
   > Set MeshToCopy(MeshToCopy:Skeletal Mesh Component Object Reference=GetOwnerSkeletalMeshComponent()) -> (Output_Get:Skeletal Mesh Component Object Reference)
   n2 = TryGetPawnOwner()
   Is Valid(exec=@n1.then_1, InputObject=@n2)
   > n3 = Cast To FortPlayerPawnAthena(Object=@n2)
   > Set playerPawn(playerPawn:Fort Player Pawn Athena Object Reference=@n3) -> (Output_Get:Fort Player Pawn Athena Object Reference)
   Event BlueprintUpdateAnimation()
   > n4 = Sequence!(execute) -> (then_0, then_1, then_2:Exec, then_3:Exec, then_4:Exec)
   n5 = GetOwnerSkeletalMeshComponent()
   n6 = GetDeltaTransformFromRefPose(self=@n5, BoneName=upperarm_l, BaseName=clavicle_l)
   > Set armUp_L(armUp_L:double=BreakRotator(InRot=BreakTransform(InTransform=@n6).Rotation).Pitch) -> (Output_Get:double)
   n7 = GetDeltaTransformFromRefPose(self=@n5, BoneName=upperarm_r, BaseName=clavicle_r)
   Set armUp_R(execute=@n4.then_1, armUp_R:double=BreakRotator(InRot=BreakTransform(InTransform=@n7).Rotation).Pitch) -> (Output_Get:double)
   n8 = "Absolute (Float)"(A=GetOwnerDirection!(self:Custom Character Part Anim Instance Object Reference) -> (ReturnValue:float))
   Set movingBwd(execute=@n4.then_2, movingBwd:double=MapRangeClamped(Value=@n8, InRangeA=150, InRangeB=180)) -> (Output_Get:double)
   > Set movingFwd(movingFwd:double=MapRangeClamped(Value=@n8, InRangeA=0, InRangeB=30, OutRangeA=1, OutRangeB=0)) -> (Output_Get:double)
   n9 = Get CurrentLOD()
   n10 = Get bRigidBodyEnabled()
   n11 = Set isRigidBodyDisabled(
    execute=@n4.then_3
    isRigidBodyDisabled:bool=OR Boolean(A=NOT Boolean(A=@n10), B="integer > integer"(A=@n9, B=1) {is=title K2Node_CallFunction})
   ) -> (Output_Get:bool)
   > n12 = Branch(Condition=@n11)
   > Set noDynamic(noDynamic:double=1) -> (Output_Get:double)
   Set noDynamic(execute=@n12.else, noDynamic=0)
   n13 = Get playerPawn() -> (playerPawn:Fort Player Pawn Athena Object Reference)
   Is Valid(exec=@n4.then_4, InputObject=@n13)
   > Do Once()
   n14 = GetSkeletalMeshForPartType(self=@n13, PartType=Backpack)
   > Set IsBackpackOn(IsBackpackOn:double="To Float (Boolean)"(InBool="Not Equal (Object)"!(A:Object Reference=@n14, B:Object Reference) -> (ReturnValue:bool))) -> (Output_Get:double)
   TryGetPawnOwner()
  ]}
]
variables [
 armUp_L {type=number}
 armUp_R {type=number}
 IsBackpackOn {type=number}
 isRigidBodyDisabled {type=boolean}
 MeshToCopy {type=SkeletalMeshComponent}
 movingBwd {type=number}
 movingFwd {type=number}
 noDynamic {type=number}
 playerPawn {type=FortPlayerPawnAthena}
]
